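# Material function that generates animated, glittering gold dust for UI from an input UV, outputting boosted GoldDust color and a Noise mask.
# Foreground and background tilings drive panned dust and rotated, panned particle samples, with local UVs offsetting the foreground.
# Cross-multiplying particle channels between layers creates sparkles tinted between two golds.
# Two panned blurred-noise samples, summed and sharpened by a power, modulate particle intensity before dust tints are added.
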MaterialFunction /OfferCatalog/Art/Majesty/MF_UI_GoldDust
graph {nodes [
  n1 = ScalarParameter() {defaultValue=1.25, group=Dust, parameterName=DustTile_A}
  n2 = FunctionInput() {inputName=UVs, inputType=FunctionInput_Vector2}
  n3 = Add(
   A=Multiply(A=@n2, B=ScalarParameter() {defaultValue=2.5, parameterName=GlobalTileForeground})
   B=Multiply(A=TextureCoordinate(), B=ScalarParameter() {defaultValue=0.25, parameterName=LocalizedUV})
  )
  n4 = TextureSample(UVs=Panner(Coordinate=Multiply(A=@n3, B=@n1)) {speedX=0.04, speedY=-0.08}) {samplerType=SAMPLERTYPE_Masks, texture=&/T_UI_Dust}
  n5 = ScalarParameter() {defaultValue=1, group=Dust, parameterName=DustTile_B}
  n6 = Multiply(A=@n2, B=ScalarParameter() {defaultValue=8, parameterName=GlobalTileBackground})
  n7 = TextureSample(UVs=Panner(Coordinate=Multiply(A=@n6, B=@n5)) {speedX=0.08, speedY=-0.24}) {samplerType=SAMPLERTYPE_Masks, texture=&T_UI_Dust}
  n8 = Constant3Vector() {constant=<rgba 0.7604 0.3477 0.0198 1>}
  n9 = Constant3Vector() {constant=<rgba 0.9323 0.7065 0.0291 1>}
  n10 = ScalarParameter() {defaultValue=1.5, group=Particles, parameterName=ParticleScale}
  n11 = Time()
  n12 = ScalarParameter() {defaultValue=-0.075, group=Particles, parameterName=ParticleRotationBackground}
  n13 = Add(A=Rotator(Coordinate=@n6, Time=Multiply(A=@n11, B=@n12)) {centerX=0, centerY=0})
  n14 = Multiply(A=Time(), B=ScalarParameter() {defaultValue=0.1125, group=Particles, parameterName=ParticleTimeSpeedFactor})
  n15 = AppendVector(
   A=ScalarParameter() {defaultValue=0.14, group=Particles, parameterName=ParticlePanBackground_X}
   B=ScalarParameter() {defaultValue=-0.2, group=Particles, parameterName=ParticlePanBackground_Y}
  )
  n16 = TextureSample(UVs=Multiply(A=@n10, B=Panner(Coordinate=@n13, Time=@n14, Speed=@n15) {speedX=0.25, speedY=0.25})) {samplerType=SAMPLERTYPE_Masks, texture=&/T_UI_Particles}
  n17 = Rotator(
   Coordinate=@n3
   Time=Multiply(A=Time(), B=ScalarParameter() {defaultValue=-0.125, group=Particles, parameterName=ParticleRotationForeground})
  ) {centerX=0, centerY=0}
  n18 = AppendVector(
   A=ScalarParameter() {defaultValue=0.07, group=Particles, parameterName=ParticlePanForeground_X}
   B=ScalarParameter() {defaultValue=-0.1, group=Particles, parameterName=ParticlePanForeground_Y}
  )
  n19 = TextureSample(UVs=Multiply(A=Panner(Coordinate=@n17, Time=@n14, Speed=@n18) {speedX=0.35, speedY=0.125}, B=@n10)) {samplerType=SAMPLERTYPE_Masks, texture=&T_UI_Particles}
  n20 = Add(A=Add(A=Multiply(A=@n16.B, B=@n19.R), B=Multiply(A=@n16.G, B=@n19.G)), B=Multiply(A=@n19.B, B=@n16.R))
  n21 = ScalarParameter() {defaultValue=0.5, group=AlphaNoise, parameterName=NoiseMultA}
  n22 = ScalarParameter() {defaultValue=0.09, group=AlphaNoise, parameterName=NoisePanA_X}
  n23 = ScalarParameter() {defaultValue=-0.03, group=AlphaNoise, parameterName=NoisePanA_Y}
  n24 = TextureSample(UVs=Panner(Coordinate=Multiply(A=@n2, B=@n21), Speed=AppendVector(A=@n22, B=@n23)) {speedX=0.09, speedY=-0.03}) {samplerType=SAMPLERTYPE_Grayscale, texture=&LowResBlurredNoise}
  n25 = ScalarParameter() {defaultValue=1.5, group=AlphaNoise, parameterName=NoiseMultB}
  n26 = ScalarParameter() {defaultValue=0.05, group=AlphaNoise, parameterName=NoisePanB_X}
  n27 = ScalarParameter() {defaultValue=-0.18, group=AlphaNoise, parameterName=NoisePanB_Y}
  n28 = TextureSample(UVs=Panner(Coordinate=Multiply(A=@n2, B=@n25), Speed=AppendVector(A=@n26, B=@n27)) {speedX=0.05, speedY=-0.18}) {samplerType=SAMPLERTYPE_Grayscale, texture=&LowResBlurredNoise}
  n29 = ScalarParameter() {defaultValue=5, group=AlphaNoise, parameterName=NoisePower}
  n30 = Multiply(A=Power(Base=Add(A=@n24.R, B=@n28.R), Exp=@n29), B=@n29)
  n31 = LinearInterpolate(
   A=ScalarParameter() {defaultValue=32, group=Particles, parameterName=ParticleIntensityLow}
   B=ScalarParameter() {defaultValue=128, group=Particles, parameterName=ParticleIntensityHigh}
   Alpha=Saturate(None=@n30)
  )
  n32 = Multiply(A=Multiply(A=LinearInterpolate(A=@n8, B=@n9, Alpha=@n20), B=@n20), B=@n31)
  n33 = Add(A=DotProduct(A=Add(A=@n4.R, B=@n7.R), B=@n32), B=@n32)
  n34 = Multiply(A=@n4.G, B=@n7.G)
  n35 = Multiply(A=@n4.B, B=@n7.B)
  n36 = Add(
   A=LinearInterpolate(B=Constant3Vector() {constant=<rgba 0.5625 0.215 0.0379 0>}, Alpha=Saturate(None=@n34))
   B=LinearInterpolate(B=Constant3Vector() {constant=<rgba 0.8229 0.3967 0.0514 0>}, Alpha=Multiply(A=@n35, B=@n31))
  )
  FunctionOutput(None=Multiply(A=Add(A=@n33, B=@n36), B=ScalarParameter() {defaultValue=4, parameterName=GoldDustBoost})) {outputName=GoldDust}
  FunctionOutput(None=@n30) {outputName=Noise}
 ]}
